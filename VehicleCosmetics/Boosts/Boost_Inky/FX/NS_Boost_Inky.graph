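# Niagara system for the Inky vehicle boost cosmetic, emitting inky cone, ribbon and trail effects from boost sockets on the car's skeletal mesh.
# A filtered socket count places cone sprites, multi-renderer ribbons and backward-moving trail sprites per port, offset by a car vector.
# Painted-variant modules choose per-color constants or over-life curves, with trail sprites darkening as they age.
# A frontend-preview flag alters ribbon width plus trail lifetimes and color curves.
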NiagaraSystem /VehicleCosmetics/Boosts/Boost_Inky/FX/NS_Boost_Inky
dependencies {
 usedDataInterfaces [
  &NiagaraDataInterfaceSpriteRendererInfo
  &NiagaraDataInterfaceCurve
  &NiagaraDataInterfaceColorCurve
  &NiagaraDataInterfaceSkeletalMesh
 ]
 usedDynamicInputs [&Multiply_LinearColorByFloat]
 usedModules [
  &self:SystemSpawnScript.SystemScriptSource.SystemScriptGraph.NiagaraNodeAssignment_0.SetVariables_#1
  &SystemState
 ]
}
emitters {
 NE_BoostCone_Sprites {
  emitterUpdateScript [
   &EmitterState(Life Cycle Mode=::NewEnumerator1, Loop Duration=1, UseLoopDelay=false)
   &VC_FilteredSocketCount()
   &SpawnRate(SpawnRate=17, Use Spawn Probability=false)
  ]
  particleSpawnScript [
   &VC_PaintedVariantsOverride(
    Black=<rgba 0.03 0.03 0.03 1>
    BurntSienna=<rgba 0.5 0.0625 0 1>
    Cobalt=<rgba 0 0 4 1>
    Crimson=<rgba 1.75 0.175 0.2013 2>
    DefaultColor=<rgba 0.4333 0.15 1 2>
    ForestGreen=<rgba 0 0.275 0 1>
    Pink=<rgba 2 0 1.3333 1>
    Purple=<rgba 0.1667 0 0.25 1>
    SkyBlue=<rgba 0 0.2917 0.5 1>
   ) {enabled=false}
   &InitializeParticle(
    Color=$Particles.Color
    Lifetime Max=1.25
    Lifetime Min=1.15
    Lifetime Mode=::NewEnumerator1
    Sprite Rotation Mode=::NewEnumerator2
    Sprite Size Mode=::NewEnumerator1
    Uniform Sprite Size Max=220
    Uniform Sprite Size Min=200
    UsePositionOffset=false
    Write Material Random=true
   )
   &PartitionParticles(Number Of Partitions=$Emitter.PortCount)
   &SkeletalMeshLocation(
    Bone / Socket Index=$Particles.PartitionParticles.Partition
    Bone / Socket Sampling Mode=::NewEnumerator5
    MaskSpawnGroup=false
    Mesh Sampling Type=::NewEnumerator6
    Sampled Position Offset=&AddVectorToPosition(Position=<-110 0 0>, Vector=$User.DTCarOffset)
    Skeletal Mesh=$System.SkeletalMesh
   )
   &ShapeLocation(MaskSpawnGroup=false, Override Seed=false, Sphere Radius=20, UseOwnerScale=true)
   &AddVelocity(
    Module.Use Falloff Distance=false
    Override Seed=false
    Velocity Mode=::NewEnumerator1
    Velocity Origin=$Particles.Position
    Velocity Speed=10
   ) {enabled=false}
   /VehicleCosmetics/FX/VC_FXParents/Modules/DLR_ProductPreviews(GameModeID:NiagaraInt32=$System.GameModeID, bPreviewVehicleSpeed:NiagaraBool=$System.bPreviewVehicleSpeed)
  ]
  particleUpdateScript [
   &ParticleState()
   &SubUVAnimation(
    "Sprite Renderer "=DataInterface() {propertyValues {SpriteRenderer=&self:NE_BoostCone_Sprites_0.NiagaraSpriteRendererProperties_0}}
    UseEndFrame=false
    UseStartFrame=false
   )
   &ScaleSpriteSize(Uniform Curve Sprite Scale=DataInterface() {propertyValues {Curve {
       defaultValue=0
       keys [
        {arriveTangent=-0.2143, interpMode=RCIM_Cubic, leaveTangent=-0.2143, tangentMode=RCTM_User, time=0, value=0}
        {arriveTangent=0.7123, interpMode=RCIM_Cubic, leaveTangent=0.7123, tangentMode=RCTM_User, time=1, value=1}
       ]
      }}})
   &VC_PaintedVariantsOverride(
    Black=&ColorFromCurve(ColorCurve=DataInterface() {propertyValues {
       AlphaCurve {keys [<at 0.1566 2 RCIM_Cubic>,<at 1 0 RCIM_Cubic>]}
       BlueCurve {keys [<at 0 0.02 RCIM_Cubic>,<at 1 0.02 RCIM_Cubic>]}
       GreenCurve {keys [<at 0 0.02 RCIM_Cubic>,<at 1 0.02 RCIM_Cubic>]}
       RedCurve {keys [<at 0 0.02 RCIM_Cubic>,<at 1 0.02 RCIM_Cubic>]}
      }})
    BurntSienna=&ColorFromCurve(ColorCurve=DataInterface() {propertyValues {
       AlphaCurve {keys [<at 0 3 RCIM_Cubic>,<at 1 0 RCIM_Cubic>]}
       BlueCurve {keys [<at 0 0 RCIM_Cubic>,<at 1 0 RCIM_Cubic>]}
       GreenCurve {keys [<at 0 0.0624 RCIM_Cubic>,<at 1 0.1842 RCIM_Cubic>]}
       RedCurve {keys [<at 0 0.25 RCIM_Cubic>,<at 1 0.65 RCIM_Cubic>]}
      }})
    Cobalt=&ColorFromCurve(ColorCurve=DataInterface() {propertyValues {
       AlphaCurve {keys [<at 0 2 RCIM_Cubic>,<at 1 0 RCIM_Cubic>]}
       BlueCurve {keys [<at 0 4 RCIM_Cubic>,<at 1 0.5 RCIM_Cubic>]}
       GreenCurve {keys [<at 0 0 RCIM_Cubic>,<at 1 0.0833 RCIM_Cubic>]}
       RedCurve {keys [<at 0 0 RCIM_Cubic>,<at 1 0 RCIM_Cubic>]}
      }})
    Crimson=&ColorFromCurve(ColorCurve=DataInterface() {propertyValues {
       AlphaCurve {keys [<at 0 3 RCIM_Cubic>,<at 0.368 1>,<at 1 0 RCIM_Cubic>]}
       BlueCurve {keys [<at 0 0.7875 RCIM_Cubic>,<at 1 0.145 RCIM_Cubic>]}
       GreenCurve {keys [<at 0 0.75 RCIM_Cubic>,<at 1 0.1 RCIM_Cubic>]}
       RedCurve {keys [<at 0 3 RCIM_Cubic>,<at 1 1 RCIM_Cubic>]}
      }})
    DefaultColor=&ColorFromCurve(ColorCurve=DataInterface() {propertyValues {
       AlphaCurve {keys [<at 0 2 RCIM_Cubic>,<at 1 0 RCIM_Cubic>]}
       BlueCurve {keys [<at 0 15 RCIM_Cubic>,<at 1 1 RCIM_Cubic>]}
       GreenCurve {keys [<at 0 0 RCIM_Cubic>,<at 1 0 RCIM_Cubic>]}
       RedCurve {keys [<at 0 2.5 RCIM_Cubic>,<at 1 0.1167 RCIM_Cubic>]}
      }})
    ForestGreen=&ColorFromCurve(ColorCurve=DataInterface() {propertyValues {
       AlphaCurve {keys [<at 0 3 RCIM_Cubic>,<at 1 0 RCIM_Cubic>]}
       BlueCurve {keys [<at 0 0 RCIM_Cubic>,<at 1 0 RCIM_Cubic>]}
       GreenCurve {keys [<at 0 0.2 RCIM_Cubic>,<at 1 0.65 RCIM_Cubic>]}
       RedCurve {keys [<at 0 0.0336 RCIM_Cubic>,<at 1 0.1092 RCIM_Cubic>]}
      }})
    Grey=&ColorFromCurve(ColorCurve=DataInterface() {propertyValues {
       AlphaCurve {keys [<at 0.1566 2 RCIM_Cubic>,<at 1 0 RCIM_Cubic>]}
       BlueCurve {keys [<at 0 0.65 RCIM_Cubic>,<at 1 0.15 RCIM_Cubic>]}
       GreenCurve {keys [<at 0 0.65 RCIM_Cubic>,<at 1 0.15 RCIM_Cubic>]}
       RedCurve {keys [<at 0 0.65 RCIM_Cubic>,<at 1 0.15 RCIM_Cubic>]}
      }})
    Lime=&ColorFromCurve(ColorCurve=DataInterface() {propertyValues {
       AlphaCurve {keys [<at 0 3 RCIM_Cubic>,<at 1 0 RCIM_Cubic>]}
       BlueCurve {keys [<at 0 0 RCIM_Cubic>,<at 1 0 RCIM_Cubic>]}
       GreenCurve {keys [<at 0 0.75 RCIM_Cubic>,<at 1 0.75 RCIM_Cubic>]}
       RedCurve {keys [<at 0 0.25 RCIM_Cubic>,<at 1 0.1875 RCIM_Cubic>]}
      }})
    Orange=&ColorFromCurve(ColorCurve=DataInterface() {propertyValues {
       AlphaCurve {keys [<at 0 2 RCIM_Cubic>,<at 1 0 RCIM_Cubic>]}
       BlueCurve {keys [<at 0 0.0625 RCIM_Cubic>,<at 1 0 RCIM_Cubic>]}
       GreenCurve {keys [<at 0 0.1615 RCIM_Cubic>,<at 1 0.0333 RCIM_Cubic>]}
       RedCurve {keys [<at 0 1.25 RCIM_Cubic>,<at 1 1 RCIM_Cubic>]}
      }})
    Pink=&ColorFromCurve(ColorCurve=DataInterface() {propertyValues {
       AlphaCurve {keys [<at 0 2 RCIM_Cubic>,<at 1 0 RCIM_Cubic>]}
       BlueCurve {keys [<at 0.5102 1.25 RCIM_Cubic>]}
       GreenCurve {keys [<at 0.5102 0 RCIM_Cubic>]}
       RedCurve {keys [<at 0.5102 2.5 RCIM_Cubic>]}
      }})
    Purple=<rgba 0.1667 0 0.25 1>
    Saffron=&ColorFromCurve(ColorCurve=DataInterface() {propertyValues {
       AlphaCurve {keys [<at 0 2 RCIM_Cubic>,<at 1 0 RCIM_Cubic>]}
       BlueCurve {keys [<at 0.5102 0 RCIM_Cubic>]}
       GreenCurve {keys [<at 0.5102 2.5 RCIM_Cubic>]}
       RedCurve {keys [<at 0.5102 2.2917 RCIM_Cubic>]}
      }})
    SkyBlue=&ColorFromCurve(ColorCurve=DataInterface() {propertyValues {
       AlphaCurve {keys [<at 0 2 RCIM_Cubic>,<at 1 0 RCIM_Cubic>]}
       BlueCurve {keys [<at 0.4254 2 RCIM_Cubic>]}
       GreenCurve {keys [<at 0.4254 1.575 RCIM_Cubic>]}
       RedCurve {keys [<at 0.4254 0.3 RCIM_Cubic>]}
      }})
    TitaniumWhite=&ColorFromCurve(ColorCurve=DataInterface() {propertyValues {
       AlphaCurve {keys [<at 0.1566 2 RCIM_Cubic>,<at 1 0 RCIM_Cubic>]}
       BlueCurve {keys [<at 0 1 RCIM_Cubic>]}
       GreenCurve {keys [<at 0 1 RCIM_Cubic>]}
       RedCurve {keys [<at 0 1 RCIM_Cubic>]}
      }})
   ) {module=VC_PaintedVariantsOverride001}
   &InheritSourceMovement(
    Applied Position Scale=&Multiply_VectorByFloat(
     Vector=<1 1 1>
     Float=&FloatFromCurve(FloatCurve=DataInterface() {propertyValues {Curve {
         defaultValue=1
         keys [
          {arriveTangent=-1.7885, interpMode=RCIM_Cubic, leaveTangent=-1.7885, tangentMode=RCTM_User, time=0.05, value=1}
          {interpMode=RCIM_Cubic, tangentMode=RCTM_User, time=1, value=0}
         ]
        }}})
    )
    ApplyForce=false
    ApplyPosition=true
    ApplyPositionFalloff=false
   ) {enabled=false, module=InheritSourceMovement001}
   &SolveForcesAndVelocity(Clamp Velocity=false, Limit Acceleration=false) {enabled=false}
  ]
  rendererClasses [&NiagaraSpriteRendererProperties]
  renderers [{rendererClass=&NiagaraSpriteRendererProperties}]
 }
 NE_Boost_Ribbon {
  emitterSpawnScript [/SetVariables(Emitter.R3_WPO:Vector4f=&MakeVector4(X=0, Y=0, Z=0, W=0)) {isSetParameters=true}]
  emitterUpdateScript [
   &EmitterState(Life Cycle Mode=::NewEnumerator1, Loop Duration=1, Recalculate Duration Each Loop=true, UseLoopDelay=false)
   &VC_FilteredSocketCount()
   &VC_RateQualityHandler(High Spawn Rate=60, Low Spawn Rate=30, Use Custom Rate?=true)
   &VC_SpawnRibbon_v1(Number of Ribbons=$Emitter.PortCount, SpawnRate=$Emitter.Spawn Rate)
  ]
  particleSpawnScript [
   &VC_PaintedVariantsOverride(
    Black=<rgba 0.05 0.05 0.05 20>
    BurntSienna=<rgba 1 0.3333 0 4>
    Cobalt=<rgba 0.075 0.075 3 1>
    Crimson=<rgba 1.5 0.675 0.6888 3>
    DefaultColor=<rgba 0.7313 0.225 2.25 7>
    ForestGreen=<rgba 0.25 1.5 0 2>
    Grey=<rgba 1 1 1 1.5>
    Lime=<rgba 0.875 1.75 0 2>
    Orange=<rgba 1 0.2083 0.05 3>
    Pink=<rgba 1 0.05 1 3>
    Purple=<rgba 0.43 0.05 1 3>
    R1 Black:LinearColor=<rgba 0.15 0.15 0.15 1>
    R1 BurntSienna:LinearColor=<rgba 0.425 0.2125 0 2.15>
    R1 Cobalt:LinearColor=<rgba 0 0 1 1.15>
    R1 Crimson:LinearColor=<rgba 0.75 0.0375 0.0494 2>
    R1 DefaultColor=<rgba 0.65 0.2 2 1>
    R1 ForestGreen:LinearColor=<rgba 0.035 0.175 0 2.5>
    R1 Grey:LinearColor=<rgba 0.05 0.05 0.05 2>
    R1 Lime:LinearColor=<rgba 0.1042 0.25 0 2>
    R1 Orange:LinearColor=<rgba 0.75 0.0625 0 2>
    R1 Pink:LinearColor=<rgba 0.75 0 0.75 2>
    R1 Purple:LinearColor=<rgba 0.86 0.1 2 2.25>
    R1 Saffron:LinearColor=<rgba 0.75 0.75 0 1.5>
    R1 SkyBlue:LinearColor=<rgba 0.0175 0.2392 0.35 5>
    R1 TitaniumWhite:LinearColor=<rgba 0.35 0.35 0.35 2>
    R2 Black:LinearColor=<rgba 0.3 0.3 0.3 100>
    R2 BurntSienna:LinearColor=<rgba 1 0.4167 0 50>
    R2 Cobalt:LinearColor=<rgba 0.1 0.1792 2 100>
    R2 Crimson:LinearColor=<rgba 5 0.75 1.1042 20>
    R2 DefaultColor:LinearColor=<rgba 4.0515 0.1184 14 2>
    R2 ForestGreen:LinearColor=<rgba 0.3833 1.15 0 30>
    R2 Grey:LinearColor=<rgba 0.7 0.7 0.7 50>
    R2 Lime:LinearColor=<rgba 0.625 1.5 0 25>
    R2 Orange:LinearColor=<rgba 2 0.3917 0.07 50>
    R2 Pink:LinearColor=<rgba 2 0 1.3333 60>
    R2 Saffron:LinearColor=<rgba 1.75 1.75 0.2625 40>
    R2 SkyBlue:LinearColor=<rgba 0 1.8298 2.75 60>
    R2 TitaniumWhite:LinearColor=<rgba 1.25 1.25 1.25 50>
    Saffron=<rgba 1 1 0.05 3>
    SkyBlue=<rgba 0 2 3 1.25>
    TitaniumWhite=<rgba 1.25 1.25 1.25 1.5>
    Use R1 Painted Overrides=true
    Use R2 Painted Overrides=true
   )
   &VC_InitializeRibbon_v1(
    Color=$Particles.Color
    Custom Offset=true
    Enable Skeletal Position=true
    Facing Axis:ENiagaraOrientationAxis=::YAxis
    Lifetime=1
    Local Vector Offset=&AddVectorToPosition(
     Position=/Niagara/DynamicInputs/TypeConversions/MakeCustomPositionFromBool(Bool:NiagaraBool=$User.bIsFrontendPreview, True Position:NiagaraPosition=<-70 0 0>, False Position:NiagaraPosition=<-70 0 0>)
     Vector=$User.DTCarOffset
    )
    Offset Amount=1
    Ribbon Width:NiagaraFloat=&MakeCustomFloatFromBool(Bool=$User.bIsFrontendPreview, True Float=195, False Float=215)
   )
   &AddVelocity(Velocity=<-1500 0 20>) {module=AddVelocity001}
   &DLR_ProductPreviews(GameModeID:NiagaraInt32=$System.GameModeID, bPreviewVehicleSpeed:NiagaraBool=$System.bPreviewVehicleSpeed)
  ]
  particleUpdateScript [
   &ParticleState()
   &InheritSourceMovement(
    Applied Position Scale=&Multiply_VectorByFloat(
     Vector=<1 1 1>
     Float=&FloatFromCurve(FloatCurve=DataInterface() {propertyValues {Curve {keys [
          {interpMode=RCIM_Cubic, leaveTangent=-0.0051, tangentMode=RCTM_Break, time=0, value=1}
          {arriveTangent=-0.247, interpMode=RCIM_Cubic, tangentMode=RCTM_Break, time=1, value=0.9}
         ]}}})
    )
    ApplyForce=false
    ApplyPosition=true
    ApplyPositionFalloff=false
   ) {module=InheritSourceMovement001}
   &InheritSourceMovement(
    Applied Position Scale=&VectorFromFloat(Value=&FloatFromCurve(FloatCurve=DataInterface() {propertyValues {Curve {
         defaultValue=1
         keys [
          {arriveTangent=-1.7885, interpMode=RCIM_Cubic, leaveTangent=-1.7885, tangentMode=RCTM_User, time=0.05, value=1}
          {interpMode=RCIM_Cubic, tangentMode=RCTM_User, time=1, value=0}
         ]
        }}}))
    ApplyForce=false
    ApplyPosition=true
    ApplyPositionFalloff=false
   ) {enabled=false}
   &SolveForcesAndVelocity(Clamp Velocity=false, Limit Acceleration=false)
   &ScaleColor(
    Scale Alpha=&FloatFromCurve(FloatCurve=DataInterface() {propertyValues {Curve {keys [<at 0 1>,<at 1 0>]}}})
    Scale RGB=&VectorFromFloat(Value=1.35)
    ScaleA=true
    ScaleRGB=true
   )
   &ScaleRibbonWidth(Ribbon Width Scale=&FloatFromCurve(FloatCurve=DataInterface() {propertyValues {Curve {defaultValue=0, keys [{tangentMode=RCTM_User, time=0, value=0.25}, {tangentMode=RCTM_User, time=1, value=1.15}]}}}))
   &VC_UpdateRibbon_v1(Enable Skeletal Position=true)
   /SetVariables(Particles.R3_RibbonWidth:NiagaraFloat=&FloatFromCurve(
     FloatCurve=DataInterface() {propertyValues {Curve {
        defaultValue=1
        keys [
         {arriveTangent=7.8558, interpMode=RCIM_Cubic, leaveTangent=7.8558, tangentMode=RCTM_User, time=0, value=0}
         {interpMode=RCIM_Cubic, tangentMode=RCTM_User, time=0.25, value=1}
         {arriveTangent=-0.1621, interpMode=RCIM_Cubic, leaveTangent=-0.1621, tangentMode=RCTM_User, time=0.5, value=1}
         {arriveTangent=-5.7111, interpMode=RCIM_Cubic, leaveTangent=-5.7111, tangentMode=RCTM_User, time=0.75, value=0}
        ]
       }}}
     Scale Curve=40
    )) {isSetParameters=true}
  ]
  rendererClasses [&NiagaraRibbonRendererProperties]
  renderers [
   {rendererClass=&NiagaraRibbonRendererProperties}
   {rendererClass=&NiagaraRibbonRendererProperties, rendererIndex=1}
   {rendererClass=&NiagaraRibbonRendererProperties, rendererIndex=2}
   {rendererClass=&NiagaraRibbonRendererProperties, rendererIndex=3}
   {rendererClass=&NiagaraRibbonRendererProperties, rendererIndex=4}
  ]
 }
 NE_Trail_Sprites {
  emitterUpdateScript [
   &EmitterState()
   &VC_FilteredSocketCount()
   &SpawnRate(SpawnRate=170, Use Spawn Probability=false)
  ]
  particleSpawnScript [
   &VC_PaintedVariantsOverride(
    Black=<rgba 0.03 0.03 0.03 1>
    BurntSienna=<rgba 0.5 0.0625 0 1>
    Cobalt=<rgba 0 0 4 1>
    Crimson=<rgba 0.15 0.0006 0 1>
    ForestGreen=<rgba 0 0.275 0 1>
    Pink=<rgba 2 0 1.3333 1>
    Purple=<rgba 0.1667 0 0.25 1>
    SkyBlue=<rgba 0 0.2917 0.5 1>
   ) {enabled=false}
   &InitializeParticle(
    Color=$Particles.Color
    Lifetime Max=&MakeCustomFloatFromBool(Bool=$User.bIsFrontendPreview, True Float=0.75, False Float=0.325)
    Lifetime Min:NiagaraFloat=&MakeCustomFloatFromBool(Bool=$User.bIsFrontendPreview, True Float=0.65, False Float=0.175)
    Lifetime Mode=::NewEnumerator1
    Sprite Rotation Mode=::NewEnumerator2
    Sprite Size Mode=::NewEnumerator1
    Uniform Sprite Size Max=180
    Uniform Sprite Size Min=150
    UsePositionOffset=false
    Write Material Random=true
   )
   &PartitionParticles(Number Of Partitions=$Emitter.PortCount)
   &SkeletalMeshLocation(
    Bone / Socket Index=$Particles.PartitionParticles.Partition
    Bone / Socket Sampling Mode=::NewEnumerator5
    MaskSpawnGroup=false
    Mesh Sampling Type=::NewEnumerator6
    Sampled Position Offset=&AddVectorToPosition(Position=<-100 0 0>, Vector=$User.DTCarOffset)
    Skeletal Mesh=$System.SkeletalMesh
   )
   &AddVelocity(Velocity=&RandomRangeVector(Minimum=<-1250 -5 20>, Maximum=<-1250 5 30>))
   &ShapeLocation(MaskSpawnGroup=false, Override Seed=false, Sphere Radius=30, UseOwnerScale=true)
   &DLR_ProductPreviews(GameModeID:NiagaraInt32=$System.GameModeID, bPreviewVehicleSpeed:NiagaraBool=$System.bPreviewVehicleSpeed)
  ]
  particleUpdateScript [
   &ParticleState()
   &VC_PaintedVariantsOverride(
    Black=&MakeCustomLinearColorFromBool(
     Bool=$User.bIsFrontendPreview
     True Color=&ColorFromCurve(ColorCurve=DataInterface() {propertyValues {
        AlphaCurve {keys [<at 0.2083 0>,<at 0.4725 1>,<at 0.7105 1>,<at 0.999 0>]}
        BlueCurve {keys [<at 0.5534 0 RCIM_Cubic>]}
        GreenCurve {keys [<at 0.5534 0 RCIM_Cubic>]}
        RedCurve {keys [<at 0.5534 0 RCIM_Cubic>]}
       }})
     False Color=&ColorFromCurve(ColorCurve=DataInterface() {propertyValues {
        AlphaCurve {keys [<at 0.3329 0>,<at 0.7689 1>,<at 0.999 0>]}
        BlueCurve {keys [<at 0.5206 0 RCIM_Cubic>]}
        GreenCurve {keys [<at 0.5206 0 RCIM_Cubic>]}
        RedCurve {keys [<at 0.5206 0 RCIM_Cubic>]}
       }})
    )
    BurntSienna=&MakeCustomLinearColorFromBool(
     Bool=$User.bIsFrontendPreview
     True Color=&ColorFromCurve(ColorCurve=DataInterface() {propertyValues {
        AlphaCurve {keys [<at 0.1102 0>,<at 0.3445 1>,<at 0.7805 1>,<at 0.999 0>]}
        BlueCurve {keys [<at 0 0.1111 RCIM_Cubic>,<at 0.3487 0 RCIM_Cubic>,<at 0.7058 0>]}
        GreenCurve {keys [<at 0 0.2592 RCIM_Cubic>,<at 0.3487 0.0083 RCIM_Cubic>,<at 0.7058 0.0002>]}
        RedCurve {keys [<at 0 1 RCIM_Cubic>,<at 0.3487 0.05 RCIM_Cubic>,<at 0.7058 0.001>]}
       }})
     False Color=&ColorFromCurve(ColorCurve=DataInterface() {propertyValues {
        AlphaCurve {keys [<at 0.3329 0>,<at 0.7689 1>,<at 0.999 0>]}
        BlueCurve {keys [
          <at 0.5206 0.0833 RCIM_Cubic>
          {arriveTangent=-0.2112, interpMode=RCIM_Cubic, leaveTangent=-0.2112, time=0.7176, value=0.0012}
          <at 0.9129 0.0004>
         ]}
        GreenCurve {keys [
          <at 0.5206 0.1944 RCIM_Cubic>
          {arriveTangent=-0.4798, interpMode=RCIM_Cubic, leaveTangent=-0.4798, time=0.7176, value=0.0177}
          <at 0.9129 0.0062>
         ]}
        RedCurve {keys [<at 0.5206 0.75 RCIM_Cubic>, {arriveTangent=-1.8225, interpMode=RCIM_Cubic, leaveTangent=-1.8225, time=0.7176, value=0.1}, <at 0.9129 0.035>]}
       }})
    )
    Cobalt=&MakeCustomLinearColorFromBool(
     Bool=$User.bIsFrontendPreview
     True Color=&ColorFromCurve(ColorCurve=DataInterface() {propertyValues {
        AlphaCurve {keys [<at 0.1102 0>,<at 0.3445 1>,<at 0.7805 1>,<at 0.999 0>]}
        BlueCurve {keys [<at 0.1583 1 RCIM_Cubic>, {interpMode=RCIM_Cubic, leaveTangent=-1.633, time=0.5439, value=0.1}, <at 0.8533 0.001>]}
        GreenCurve {keys [<at 0.1583 0.1111 RCIM_Cubic>, {interpMode=RCIM_Cubic, leaveTangent=-0.1815, time=0.5439, value=0}, <at 0.8533 0>]}
        RedCurve {keys [<at 0.1583 0.1111 RCIM_Cubic>, {interpMode=RCIM_Cubic, leaveTangent=-0.1815, time=0.5439, value=0}, <at 0.8533 0>]}
       }})
     False Color=&ColorFromCurve(ColorCurve=DataInterface() {propertyValues {
        AlphaCurve {keys [<at 0.3329 0>,<at 0.7689 1>,<at 0.999 0>]}
        BlueCurve {keys [<at 0.5206 0.75 RCIM_Cubic>, {arriveTangent=-1.8225, interpMode=RCIM_Cubic, leaveTangent=-1.8225, time=0.7176, value=0.1}, <at 0.9129 0.035>]}
        GreenCurve {keys [
          <at 0.5206 0.0833 RCIM_Cubic>
          {arriveTangent=-0.2112, interpMode=RCIM_Cubic, leaveTangent=-0.2112, time=0.7176, value=0.0012}
          <at 0.9129 0.0004>
         ]}
        RedCurve {keys [
          <at 0.5206 0.0833 RCIM_Cubic>
          {arriveTangent=-0.2112, interpMode=RCIM_Cubic, leaveTangent=-0.2112, time=0.7176, value=0.0012}
          <at 0.9129 0.0004>
         ]}
       }})
    )
    Crimson=&MakeCustomLinearColorFromBool(
     Bool=$User.bIsFrontendPreview
     True Color=&ColorFromCurve(ColorCurve=DataInterface() {propertyValues {
        AlphaCurve {keys [<at 0.1102 0>,<at 0.3445 1>,<at 0.7805 1>,<at 0.999 0>]}
        BlueCurve {keys [
          <at 0.2146 0.1259 RCIM_Cubic>
          {arriveTangent=-0.1905, interpMode=RCIM_Cubic, leaveTangent=-0.1905, time=0.5606, value=0.0025}
          <at 0.8676 0.0014>
         ]}
        GreenCurve {keys [
          <at 0.2146 0.1111 RCIM_Cubic>
          {arriveTangent=-0.1691, interpMode=RCIM_Cubic, leaveTangent=-0.1691, time=0.5606, value=0}
          <at 0.8676 0.0006>
         ]}
        RedCurve {keys [<at 0.2146 1 RCIM_Cubic>, {arriveTangent=-1.4547, interpMode=RCIM_Cubic, leaveTangent=-1.4547, time=0.5606, value=0.15}, <at 0.8676 0.05>]}
       }})
     False Color=&ColorFromCurve(ColorCurve=DataInterface() {propertyValues {
        AlphaCurve {keys [<at 0.3329 0>,<at 0.7689 1>,<at 0.999 0>]}
        BlueCurve {keys [
          <at 0.5206 0.0944 RCIM_Cubic>
          {arriveTangent=-0.2381, interpMode=RCIM_Cubic, leaveTangent=-0.2381, time=0.7176, value=0.0029}
          <at 0.9129 0.001>
         ]}
        GreenCurve {keys [
          <at 0.5206 0.0833 RCIM_Cubic>
          {arriveTangent=-0.2112, interpMode=RCIM_Cubic, leaveTangent=-0.2112, time=0.7176, value=0.0012}
          <at 0.9129 0.0004>
         ]}
        RedCurve {keys [<at 0.5206 0.75 RCIM_Cubic>, {arriveTangent=-1.8225, interpMode=RCIM_Cubic, leaveTangent=-1.8225, time=0.7176, value=0.1}, <at 0.9129 0.035>]}
       }})
    )
    DefaultColor=&MakeCustomLinearColorFromBool(
     Bool=$User.bIsFrontendPreview
     True Color=&ColorFromCurve(ColorCurve=DataInterface() {propertyValues {
        AlphaCurve {keys [<at 0.1102 0>,<at 0.3445 1>,<at 0.7805 1>,<at 0.999 0>]}
        BlueCurve {keys [<at 0.2146 1 RCIM_Cubic>, {interpMode=RCIM_Cubic, leaveTangent=-1.4171, time=0.4419, value=0.1}, <at 0.8385 0.05>]}
        GreenCurve {keys [<at 0.2146 0.1111 RCIM_Cubic>, {interpMode=RCIM_Cubic, leaveTangent=-0.1648, time=0.4419, value=0.0012}, <at 0.8385 0.0006>]}
        RedCurve {keys [<at 0.2146 0.1851 RCIM_Cubic>, {interpMode=RCIM_Cubic, leaveTangent=-0.2598, time=0.4419, value=0.022}, <at 0.8385 0.011>]}
       }})
     False Color=&ColorFromCurve(ColorCurve=DataInterface() {propertyValues {
        AlphaCurve {keys [<at 0.3329 0>,<at 0.7689 1>,<at 0.999 0>]}
        BlueCurve {keys [<at 0.5206 0.75 RCIM_Cubic>, {interpMode=RCIM_Cubic, leaveTangent=-1.6423, time=0.7176, value=0.1}, <at 0.9129 0.035>]}
        GreenCurve {keys [<at 0.5206 0.0833 RCIM_Cubic>, {interpMode=RCIM_Cubic, leaveTangent=-0.1903, time=0.7176, value=0.0012}, <at 0.9129 0.0004>]}
        RedCurve {keys [<at 0.5206 0.1389 RCIM_Cubic>, {interpMode=RCIM_Cubic, leaveTangent=-0.3013, time=0.7176, value=0.022}, <at 0.9129 0.0077>]}
       }})
    )
    ForestGreen=&MakeCustomLinearColorFromBool(
     Bool=$User.bIsFrontendPreview
     True Color=&ColorFromCurve(ColorCurve=DataInterface() {propertyValues {
        AlphaCurve {keys [<at 0.1102 0>,<at 0.3445 1>,<at 0.7805 1>,<at 0.999 0>]}
        BlueCurve {keys [<at 0.2146 0 RCIM_Cubic>, {arriveTangent=0.0002, interpMode=RCIM_Cubic, leaveTangent=0.0002, time=0.4273, value=0}, <at 1 0.0001>]}
        GreenCurve {keys [<at 0.2146 1 RCIM_Cubic>, {arriveTangent=-1.2604, interpMode=RCIM_Cubic, leaveTangent=-1.2604, time=0.4273, value=0.1}, <at 1 0.01>]}
        RedCurve {keys [
          <at 0.2146 0.1667 RCIM_Cubic>
          {arriveTangent=-0.2099, interpMode=RCIM_Cubic, leaveTangent=-0.2099, time=0.4273, value=0.0167}
          <at 1 0.0018>
         ]}
       }})
     False Color=&ColorFromCurve(ColorCurve=DataInterface() {propertyValues {
        AlphaCurve {keys [<at 0.3329 0>,<at 0.7689 1>,<at 0.999 0>]}
        BlueCurve {keys [<at 0.5206 0 RCIM_Cubic>, {arriveTangent=0.0011, interpMode=RCIM_Cubic, leaveTangent=0.0011, time=0.7176, value=0.0012}, <at 0.9129 0.0004>]}
        GreenCurve {keys [<at 0.5206 0.75 RCIM_Cubic>, {arriveTangent=-1.8225, interpMode=RCIM_Cubic, leaveTangent=-1.8225, time=0.7176, value=0.1}, <at 0.9129 0.035>]}
        RedCurve {keys [
          <at 0.5206 0.125 RCIM_Cubic>
          {arriveTangent=-0.3028, interpMode=RCIM_Cubic, leaveTangent=-0.3028, time=0.7176, value=0.0177}
          <at 0.9129 0.0062>
         ]}
       }})
    )
    Grey=&MakeCustomLinearColorFromBool(
     Bool=$User.bIsFrontendPreview
     True Color=&ColorFromCurve(ColorCurve=DataInterface() {propertyValues {
        AlphaCurve {keys [<at 0.2083 0>,<at 0.4725 1>,<at 0.7105 1>,<at 0.999 0>]}
        BlueCurve {keys [<at 0.5534 0.007 RCIM_Cubic>]}
        GreenCurve {keys [<at 0.5534 0.007 RCIM_Cubic>]}
        RedCurve {keys [<at 0.5534 0.007 RCIM_Cubic>]}
       }})
     False Color=&ColorFromCurve(ColorCurve=DataInterface() {propertyValues {
        AlphaCurve {keys [<at 0.3329 0>,<at 0.7689 1>,<at 0.999 0>]}
        BlueCurve {keys [<at 0.5463 0.007 RCIM_Cubic>]}
        GreenCurve {keys [<at 0.5463 0.007 RCIM_Cubic>]}
        RedCurve {keys [<at 0.5463 0.007 RCIM_Cubic>]}
       }})
    )
    Lime=&MakeCustomLinearColorFromBool(
     Bool=$User.bIsFrontendPreview
     True Color=&ColorFromCurve(ColorCurve=DataInterface() {propertyValues {
        AlphaCurve {keys [<at 0.1102 0>,<at 0.3445 1>,<at 0.7805 1>,<at 0.999 0>]}
        BlueCurve {keys [<at 0.2146 0 RCIM_Cubic>, {arriveTangent=0.0002, interpMode=RCIM_Cubic, leaveTangent=0.0002, time=0.4273, value=0}, <at 1 0.0001>]}
        GreenCurve {keys [<at 0.2146 1 RCIM_Cubic>, {arriveTangent=-1.2604, interpMode=RCIM_Cubic, leaveTangent=-1.2604, time=0.4273, value=0.1}, <at 1 0.01>]}
        RedCurve {keys [
          <at 0.2146 0.4167 RCIM_Cubic>
          {arriveTangent=-0.5282, interpMode=RCIM_Cubic, leaveTangent=-0.5282, time=0.4273, value=0.0417}
          <at 1 0.0018>
         ]}
       }})
     False Color=&ColorFromCurve(ColorCurve=DataInterface() {propertyValues {
        AlphaCurve {keys [<at 0.3329 0>,<at 0.7689 1>,<at 1 0>]}
        BlueCurve {keys [<at 0.5206 0 RCIM_Cubic>, {interpMode=RCIM_Cubic, leaveTangent=0.0011, time=0.7581, value=0.0012}, <at 1 0.0004>]}
        GreenCurve {keys [<at 0.5206 0.75 RCIM_Cubic>, {interpMode=RCIM_Cubic, leaveTangent=-1.8225, time=0.7581, value=0.1}, <at 1 0.035>]}
        RedCurve {keys [<at 0.5206 0.3125 RCIM_Cubic>, {interpMode=RCIM_Cubic, leaveTangent=-0.7808, time=0.7581, value=0.0424}, <at 1 0.0062>]}
       }})
    )
    Orange=&MakeCustomLinearColorFromBool(
     Bool=$User.bIsFrontendPreview
     True Color=&ColorFromCurve(ColorCurve=DataInterface() {propertyValues {
        AlphaCurve {keys [<at 0.1102 0>,<at 0.3445 1>,<at 0.7805 1>,<at 0.999 0>]}
        BlueCurve {keys [<at 0.2146 0.05 RCIM_Cubic>, {arriveTangent=-0.0761, interpMode=RCIM_Cubic, leaveTangent=-0.0761, time=0.5606, value=0}, <at 0.8676 0.0003>]}
        GreenCurve {keys [
          <at 0.2146 0.1292 RCIM_Cubic>
          {arriveTangent=-0.191, interpMode=RCIM_Cubic, leaveTangent=-0.191, time=0.5606, value=0.0167}
          <at 0.8676 0.0044>
         ]}
        RedCurve {keys [<at 0.2146 1 RCIM_Cubic>, {arriveTangent=-1.493, interpMode=RCIM_Cubic, leaveTangent=-1.493, time=0.5606, value=0.1}, <at 0.8676 0.025>]}
       }})
     False Color=&ColorFromCurve(ColorCurve=DataInterface() {propertyValues {
        AlphaCurve {keys [<at 0.3329 0>,<at 0.7689 1>,<at 0.999 0>]}
        BlueCurve {keys [<at 0.5206 0.0375 RCIM_Cubic>, {arriveTangent=-0.0956, interpMode=RCIM_Cubic, leaveTangent=-0.0956, time=0.7653, value=0}, <at 0.9129 0>]}
        GreenCurve {keys [
          <at 0.5206 0.0969 RCIM_Cubic>
          {arriveTangent=-0.2363, interpMode=RCIM_Cubic, leaveTangent=-0.2363, time=0.7653, value=0.0083}
          <at 0.9129 0.0042>
         ]}
        RedCurve {keys [<at 0.5206 0.75 RCIM_Cubic>, {arriveTangent=-1.848, interpMode=RCIM_Cubic, leaveTangent=-1.848, time=0.7653, value=0.05}, <at 0.9129 0.025>]}
       }})
    )
    Pink=&MakeCustomLinearColorFromBool(
     Bool=$User.bIsFrontendPreview
     True Color=&ColorFromCurve(ColorCurve=DataInterface() {propertyValues {
        AlphaCurve {keys [<at 0.1102 0>,<at 0.3445 1>,<at 0.7805 1>,<at 0.999 0>]}
        BlueCurve {keys [<at 0.2146 0.1292 RCIM_Cubic>, {arriveTangent=-0.1978, interpMode=RCIM_Cubic, leaveTangent=-0.1978, time=0.5606, value=0}, <at 0.8676 0>]}
        GreenCurve {keys [<at 0.2146 0.05 RCIM_Cubic>, {arriveTangent=-0.074, interpMode=RCIM_Cubic, leaveTangent=-0.074, time=0.5606, value=0}, <at 0.8676 0.0017>]}
        RedCurve {keys [<at 0.2146 1 RCIM_Cubic>, {arriveTangent=-1.4547, interpMode=RCIM_Cubic, leaveTangent=-1.4547, time=0.5606, value=0.1}, <at 0.8676 0.05>]}
       }})
     False Color=&ColorFromCurve(ColorCurve=DataInterface() {propertyValues {
        AlphaCurve {keys [<at 0.3329 0>,<at 0.7689 1>,<at 0.999 0>]}
        BlueCurve {keys [<at 0.5206 0.0969 RCIM_Cubic>, {arriveTangent=-0.2469, interpMode=RCIM_Cubic, leaveTangent=-0.2469, time=0.7653, value=0}, <at 0.9129 0>]}
        GreenCurve {keys [
          <at 0.5206 0.0375 RCIM_Cubic>
          {arriveTangent=-0.0935, interpMode=RCIM_Cubic, leaveTangent=-0.0935, time=0.7653, value=0}
          <at 0.9129 0.0008>
         ]}
        RedCurve {keys [<at 0.5206 0.75 RCIM_Cubic>, {arriveTangent=-1.848, interpMode=RCIM_Cubic, leaveTangent=-1.848, time=0.7653, value=0.05}, <at 0.9129 0.025>]}
       }})
    )
    Purple=<rgba 0.1667 0 0.25 1>
    Saffron=&MakeCustomLinearColorFromBool(
     Bool=$User.bIsFrontendPreview
     True Color=&ColorFromCurve(ColorCurve=DataInterface() {propertyValues {
        AlphaCurve {keys [<at 0.1102 0>,<at 0.3445 1>,<at 0.7805 1>,<at 0.999 0>]}
        BlueCurve {keys [<at 0.2146 0.1111 RCIM_Cubic>, {arriveTangent=-0.1414, interpMode=RCIM_Cubic, leaveTangent=-0.1414, time=0.532, value=0.0025}, <at 1 0>]}
        GreenCurve {keys [
          <at 0.2146 0.8518 RCIM_Cubic>
          {arriveTangent=-0.4111, interpMode=RCIM_Cubic, leaveTangent=-0.4111, tangentMode=RCTM_User, time=0.532, value=0.0421}
          <at 1 0.0417>
         ]}
        RedCurve {keys [
          <at 0.2146 1 RCIM_Cubic>
          {arriveTangent=-0.5858, interpMode=RCIM_Cubic, leaveTangent=-0.5858, tangentMode=RCTM_User, time=0.532, value=0.05}
          <at 1 0.05>
         ]}
       }})
     False Color=&ColorFromCurve(ColorCurve=DataInterface() {propertyValues {
        AlphaCurve {keys [<at 0.3329 0>,<at 0.7689 1>,<at 0.999 0>]}
        BlueCurve {keys [
          <at 0.5206 0.0833 RCIM_Cubic>
          {arriveTangent=-0.0181, interpMode=RCIM_Cubic, leaveTangent=-0.0181, tangentMode=RCTM_User, time=0.7176, value=0.0012}
          <at 0.9129 0>
         ]}
        GreenCurve {keys [
          <at 0.5206 0.6389 RCIM_Cubic>
          {arriveTangent=-1.0628, interpMode=RCIM_Cubic, leaveTangent=-1.0628, tangentMode=RCTM_User, time=0.7167, value=0.1255}
          <at 0.9129 0.0292>
         ]}
        RedCurve {keys [
          <at 0.5206 0.75 RCIM_Cubic>
          {arriveTangent=-1.2626, interpMode=RCIM_Cubic, leaveTangent=-1.2626, tangentMode=RCTM_User, time=0.7167, value=0.1419}
          <at 0.9129 0.035>
         ]}
       }})
    )
    SkyBlue=&MakeCustomLinearColorFromBool(
     Bool=$User.bIsFrontendPreview
     True Color=&ColorFromCurve(ColorCurve=DataInterface() {propertyValues {
        AlphaCurve {keys [<at 0.1102 0>,<at 0.3445 1>,<at 0.7805 1>,<at 0.999 0>]}
        BlueCurve {keys [<at 0.2146 1 RCIM_Cubic>, {arriveTangent=-1.5225, interpMode=RCIM_Cubic, leaveTangent=-1.5225, time=0.4419, value=0.1}, <at 0.8385 0.05>]}
        GreenCurve {keys [
          <at 0.2146 0.7778 RCIM_Cubic>
          {arriveTangent=-1.1861, interpMode=RCIM_Cubic, leaveTangent=-1.1861, time=0.4419, value=0.0753}
          <at 0.8385 0.0377>
         ]}
        RedCurve {keys [
          <at 0.2146 0.1111 RCIM_Cubic>
          {arriveTangent=-0.177, interpMode=RCIM_Cubic, leaveTangent=-0.177, time=0.4419, value=0.0012}
          <at 0.8385 0.0006>
         ]}
       }})
     False Color=&ColorFromCurve(ColorCurve=DataInterface() {propertyValues {
        AlphaCurve {keys [<at 0.3329 0>,<at 0.7689 1>,<at 0.999 0>]}
        BlueCurve {keys [<at 0.5206 0.75 RCIM_Cubic>, {arriveTangent=-1.8225, interpMode=RCIM_Cubic, leaveTangent=-1.8225, time=0.7176, value=0.1}, <at 0.9129 0.035>]}
        GreenCurve {keys [
          <at 0.5206 0.5833 RCIM_Cubic>
          {arriveTangent=-1.4197, interpMode=RCIM_Cubic, leaveTangent=-1.4197, time=0.7176, value=0.0753}
          <at 0.9129 0.0264>
         ]}
        RedCurve {keys [
          <at 0.5206 0.0833 RCIM_Cubic>
          {arriveTangent=-0.2112, interpMode=RCIM_Cubic, leaveTangent=-0.2112, time=0.7176, value=0.0012}
          <at 0.9129 0.0004>
         ]}
       }})
    )
    TitaniumWhite=&MakeCustomLinearColorFromBool(
     Bool=$User.bIsFrontendPreview
     True Color=&ColorFromCurve(ColorCurve=DataInterface() {propertyValues {
        AlphaCurve {keys [<at 0.2083 0>,<at 0.4725 1>,<at 0.7105 1>,<at 0.999 0>]}
        BlueCurve {keys [<at 0.0702 0.025>,<at 0.4035 0.0001 RCIM_Cubic>]}
        GreenCurve {keys [<at 0.0702 0.025>,<at 0.4035 0.0001 RCIM_Cubic>]}
        RedCurve {keys [<at 0.0702 0.025>,<at 0.4035 0.0001 RCIM_Cubic>]}
       }})
     False Color=&ColorFromCurve(ColorCurve=DataInterface() {propertyValues {
        AlphaCurve {keys [<at 0.3329 0>,<at 0.7689 1>,<at 0.999 0>]}
        BlueCurve {keys [<at 0.5463 0.007 RCIM_Cubic>]}
        GreenCurve {keys [<at 0.5463 0.007 RCIM_Cubic>]}
        RedCurve {keys [<at 0.5463 0.007 RCIM_Cubic>]}
       }})
    )
   ) {module=VC_PaintedVariantsOverride001}
   &SubUVAnimation(
    "Sprite Renderer "=DataInterface() {propertyValues {SpriteRenderer=&self:NE_Trail_Sprites_0.NiagaraSpriteRendererProperties_0}}
    UseEndFrame=false
    UseStartFrame=false
   )
   &ScaleSpriteSize()
   &InheritSourceMovement(
    Applied Position Scale=&Multiply_VectorByFloat(
     Vector=<1 1 1>
     Float=&FloatFromCurve(FloatCurve=DataInterface() {propertyValues {Curve {
         defaultValue=1
         keys [
          {arriveTangent=-1.7885, interpMode=RCIM_Cubic, leaveTangent=-1.7885, tangentMode=RCTM_User, time=0.05, value=1}
          {interpMode=RCIM_Cubic, tangentMode=RCTM_User, time=1, value=0}
         ]
        }}})
    )
    ApplyForce=false
    ApplyPosition=true
    ApplyPositionFalloff=false
   ) {enabled=false, module=InheritSourceMovement001}
   &SolveForcesAndVelocity(Clamp Velocity=false, Limit Acceleration=false)
  ]
  rendererClasses [&NiagaraSpriteRendererProperties]
  renderers [{rendererClass=&NiagaraSpriteRendererProperties}]
 }
}
summary {
 systemName=NS_Boost_Inky
 emitters [
  {emitterName=NE_BoostCone_Sprites, rendererClasses [&NiagaraSpriteRendererProperties]}
  {emitterName=NE_Boost_Ribbon, rendererClasses [&NiagaraRibbonRendererProperties]}
  {emitterName=NE_Trail_Sprites, rendererClasses [&NiagaraSpriteRendererProperties]}
 ]
 userVariables [
  User.bIsFrontendPreview {type=NiagaraBool}
  User.bIsPainted {type=NiagaraBool}
  User.ColorVariantID {type=NiagaraInt32}
  User.DTCarOffset {type=Vector3f, description="", defaultValue {struct=Vector3f, value=<0 0 0>}}
  User.GlowOverride {type=NiagaraFloat, defaultValue=4, description=""}
  User.PaintedColor {type=LinearColor, description="", defaultValue {struct=LinearColor, value=<rgba 1 1 1 1>}}
  User.PaintedScalar {type=NiagaraFloat, defaultValue=1, description=""}
  User.PortCount {type=NiagaraInt32, defaultValue=1}
 ]
}
